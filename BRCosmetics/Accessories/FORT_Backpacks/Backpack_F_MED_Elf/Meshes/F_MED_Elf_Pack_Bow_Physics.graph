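# Physics asset for the bow on the Elf backpack cosmetic, letting its left and right dangling pieces move with simulated physics.
# Capsule bodies cover the pack and two bones per side, with each side chained from the pack through its first bone to its second.
# Swing and twist limits are asymmetric, with the right chain allowed far more twist than the left.
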PhysicsAsset /BRCosmetics/Accessories/FORT_Backpacks/Backpack_F_MED_Elf/Meshes/F_MED_Elf_Pack_Bow_Physics
bodies {
 dyn_bow_01_l [{shapeType=Capsule, length=1.25, radius=1.75, center=<0.3 0 0>, rotation=<rot 90 0 0>}]
 dyn_bow_01_r [{shapeType=Capsule, length=1.75, radius=1.5, center=<0.35 0 0>, rotation=<rot 90 0 0>}]
 dyn_bow_02_l [{shapeType=Capsule, length=0.75, radius=1.25, center=<0.9 0 0>, rotation=<rot 90 0 0>}]
 dyn_bow_02_r [{shapeType=Capsule, length=0.75, radius=1.5, center=<0.8 0 0>, rotation=<rot 90 0 0>}]
 pack [{shapeType=Capsule, length=3.75, radius=5.5, center=<0 -6.5424 -5.636>}]
}
constraints [
 {bone1Name=dyn_bow_01_l, bone2Name=pack, swing1LimitDegrees=15, swing2LimitDegrees=20, twistLimitDegrees=5}
 {bone1Name=dyn_bow_02_l, bone2Name=dyn_bow_01_l, swing1LimitDegrees=25, swing2LimitDegrees=35, twistLimitDegrees=5}
 {bone1Name=dyn_bow_01_r, bone2Name=pack, swing1LimitDegrees=18, swing2LimitDegrees=13, twistLimitDegrees=45}
 {bone1Name=dyn_bow_02_r, bone2Name=dyn_bow_01_r, swing1LimitDegrees=18, swing2LimitDegrees=22, twistLimitDegrees=45}
]
