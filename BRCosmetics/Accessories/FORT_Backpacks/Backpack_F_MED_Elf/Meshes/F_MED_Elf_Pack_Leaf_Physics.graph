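# Physics asset for the leaf parts of the F_MED_Elf backpack cosmetic, giving its left and right leaf bone chains collision and joint limits.
# Each chain has four capsule bodies, mostly shrinking toward the tip, linked parent to child and rooted on a small capsule at the pack bone.
# Swing and tight twist limits on each joint restrict bending, with slightly different capsule sizes and swing values between the two sides.
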PhysicsAsset /BRCosmetics/Accessories/FORT_Backpacks/Backpack_F_MED_Elf/Meshes/F_MED_Elf_Pack_Leaf_Physics
bodies {
 dyn_leaf_01_l [{shapeType=Capsule, length=2.5, radius=2.75, center=<0.1 0 0>, rotation=<rot 90 0 0>}]
 dyn_leaf_01_r [{shapeType=Capsule, length=2.5, radius=2.75, center=<0.1 0 0>, rotation=<rot 90 0 0>}]
 dyn_leaf_02_l [{shapeType=Capsule, length=1.75, radius=2, center=<0.3 0 0>, rotation=<rot 90 0 0>}]
 dyn_leaf_02_r [{shapeType=Capsule, length=1.75, radius=2.5, center=<0.3 0 0>, rotation=<rot 90 0 0>}]
 dyn_leaf_03_l [{shapeType=Capsule, length=1, radius=1.5, center=<0.6 0 0>, rotation=<rot 90 0 0>}]
 dyn_leaf_03_r [{shapeType=Capsule, length=0.75, radius=2.25, center=<0.75 0 0>, rotation=<rot 90 0 0>}]
 dyn_leaf_04_l [{shapeType=Capsule, length=1.25, radius=1.25, center=<0.8 0 0>, rotation=<rot 90 0 0>}]
 dyn_leaf_04_r [{shapeType=Capsule, length=0.75, radius=1.75, center=<1 0 0>, rotation=<rot 90 0 0>}]
 pack [{shapeType=Capsule, length=1, radius=1}]
}
constraints [
 {bone1Name=dyn_leaf_01_l, bone2Name=pack, swing1LimitDegrees=15, swing2LimitDegrees=10, twistLimitDegrees=5}
 {bone1Name=dyn_leaf_02_l, bone2Name=dyn_leaf_01_l, swing1LimitDegrees=15, swing2LimitDegrees=20.1429, twistLimitDegrees=5}
 {bone1Name=dyn_leaf_03_l, bone2Name=dyn_leaf_02_l, swing1LimitDegrees=15, swing2LimitDegrees=20, twistLimitDegrees=5}
 {bone1Name=dyn_leaf_04_l, bone2Name=dyn_leaf_03_l, swing1LimitDegrees=15, swing2LimitDegrees=20, twistLimitDegrees=5}
 {bone1Name=dyn_leaf_01_r, bone2Name=pack, swing1LimitDegrees=15, swing2LimitDegrees=10, twistLimitDegrees=5}
 {bone1Name=dyn_leaf_02_r, bone2Name=dyn_leaf_01_r, swing1LimitDegrees=15, swing2LimitDegrees=20.5714, twistLimitDegrees=5}
 {bone1Name=dyn_leaf_03_r, bone2Name=dyn_leaf_02_r, swing1LimitDegrees=15, swing2LimitDegrees=30, twistLimitDegrees=5}
 {bone1Name=dyn_leaf_04_r, bone2Name=dyn_leaf_03_r, swing1LimitDegrees=15, swing2LimitDegrees=20, twistLimitDegrees=5}
]
