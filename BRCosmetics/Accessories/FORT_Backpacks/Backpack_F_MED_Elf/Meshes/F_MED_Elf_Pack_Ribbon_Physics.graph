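# Physics asset for the F_MED_Elf backpack cosmetic, defining simulation bodies for its two ribbons and bell.
# Capsule bodies cover the pack, the bell and three bones in each ribbon, with thicker capsules on the left ribbon.
# Each ribbon forms a chain hanging from the pack, while the bell attaches to the pack directly.
# Constraints keep twist tight throughout and allow wider swing below each ribbon's root joint.
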PhysicsAsset /BRCosmetics/Accessories/FORT_Backpacks/Backpack_F_MED_Elf/Meshes/F_MED_Elf_Pack_Ribbon_Physics
bodies {
 bell [{shapeType=Capsule, length=7.8209, radius=3.0765, center=<7 0 0>, rotation=<rot 90 0 0>}]
 dyn_ribbon_01_l [{shapeType=Capsule, length=1.25, radius=1.25, center=<0.8 0 0>, rotation=<rot 90 0 0>}]
 dyn_ribbon_01_r [{shapeType=Capsule, length=1.25, radius=0.75, center=<0.1 0 0>, rotation=<rot 90 0 0>}]
 dyn_ribbon_02_l [{shapeType=Capsule, length=1.25, radius=1.25, center=<1.25 0 0>, rotation=<rot 90 0 0>}]
 dyn_ribbon_02_r [{shapeType=Capsule, length=1.25, radius=0.75, center=<0.3 0 0>, rotation=<rot 90 0 0>}]
 dyn_ribbon_03_l [{shapeType=Capsule, length=1.25, radius=1.25, center=<1.3 0 0>, rotation=<rot 90 0 0>}]
 dyn_ribbon_03_r [{shapeType=Capsule, length=1.25, radius=0.75, center=<0.75 0 0>, rotation=<rot 90 0 0>}]
 pack [{shapeType=Capsule, length=3.75, radius=5.5, center=<0 -6.5424 -5.636>}]
}
constraints [
 {bone1Name=dyn_ribbon_01_l, bone2Name=pack, swing1LimitDegrees=30, swing2LimitDegrees=15, twistLimitDegrees=5}
 {bone1Name=dyn_ribbon_02_l, bone2Name=dyn_ribbon_01_l, swing1LimitDegrees=45, swing2LimitDegrees=45, twistLimitDegrees=5}
 {bone1Name=dyn_ribbon_03_l, bone2Name=dyn_ribbon_02_l, swing1LimitDegrees=45, swing2LimitDegrees=35, twistLimitDegrees=5}
 {bone1Name=dyn_ribbon_01_r, bone2Name=pack, swing1LimitDegrees=30, swing2LimitDegrees=8, twistLimitDegrees=5}
 {bone1Name=dyn_ribbon_02_r, bone2Name=dyn_ribbon_01_r, swing1LimitDegrees=45, swing2LimitDegrees=45, twistLimitDegrees=5}
 {bone1Name=dyn_ribbon_03_r, bone2Name=dyn_ribbon_02_r, swing1LimitDegrees=45, swing2LimitDegrees=20, twistLimitDegrees=5}
 {bone1Name=bell, bone2Name=pack, swing1LimitDegrees=35, swing2LimitDegrees=35, twistLimitDegrees=5}
]
